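# Additive, unlit, two-sided material for the cone glow effect on the Season 21 umbrella glider, used on Niagara mesh particles.
# Emissive layers two panning, distorted line-star textures with a vertical gradient, tinted by a bottom color.
# Alpha combines panning radial-gradient streaks, view-angle fade, a sine-pulsed vertical falloff and particle alpha.
# Caustic-texture distortion becomes a constant on low quality, mobile and forward paths; a dynamic parameter randomizes UV offset.
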Material /BRCosmetics/Gadgets/Assets/VinderTech_Umbrella/Umbrella_Season21/FX/M_S21_Umbrella_ConeGlow
graph {
 properties {
  bEnableResponsiveAA=true
  bUsedWithNiagaraMeshParticles=true
  blendMode=BLEND_Additive
  pixelDepthOffsetMode=PDOM_Legacy
  shadingModel=MSM_Unlit
  translucencyPass=MTP_BeforeDOF
  twoSided=true
  usedShadingModels=Unlit
 }
 nodes [
  Constant()
  n1 = Multiply(A=ComponentMask(None=NamedRerouteUsage()) {g=true, r=true}) {constB=0.06}
  n2 = VectorParameter() {parameterName=ColorRibsUV_Pan, defaultValue=<rgba 0.1 0.7 0 0>}
  n3 = AppendVector(A=@n2.R, B=@n2.R)
  n4 = NamedRerouteUsage()
  n5 = Panner(Coordinate=Add(A=NamedRerouteUsage(), B=@n1), Time=@n4, Speed=@n3) {speedX=1, speedY=1}
  n6 = Multiply(A=NamedRerouteUsage()) {constB=1.7}
  n7 = VectorParameter() {parameterName=ColorRibsUv2_Pan, defaultValue=<rgba -0.05 -0.5 0 1>}
  n8 = AppendVector(A=@n7.R, B=@n7.R)
  n9 = TextureSampleParameter2D(UVs=Panner(Coordinate=Subtract(A=@n6, B=@n1), Time=@n4, Speed=Multiply(A=@n3, B=@n8)) {speedX=1, speedY=1}) {parameterName=LinesTexture02, texture=/Game/Athena/Environments/Nexus/Textures/T_ZeroPoint_Line_Stars}
  n10 = VectorParameter() {parameterName=ColorRibs_Tint, defaultValue=<rgba 1 1 1 1>}
  n11 = Multiply(A=Add(
    A=TextureSampleParameter2D(UVs=@n5) {parameterName=LinesTexture01, texture=&T_ZeroPoint_Line_Stars}.RGB
    B=Multiply(A=@n9.RGB, B=@n10.RGB) {constB=0.5}
   )) {constB=2}
  n12 = Multiply(A=NamedRerouteUsage()) {constB=0.5}
  n13 = MaterialFunctionCall() {materialFunction=&LinearGradient}
  n14 = Multiply(
   A=VectorParameter() {parameterName=Color_Bottom, defaultValue=<rgba 0.3542 0.7653 1 1>}.RGB
   B=Add(A=Add(A=@n11, B=@n12), B=Power(Base=@n13.VGradient) {constExponent=3})
  )
  MaterialFunctionCall(Emissive=@n14) {materialFunction=&MF_FX_EmissiveAdjustments}
  n15 = TextureCoordinate() {uTiling=0.2}
  n16 = Time()
  n17 = MaterialFunctionCall(
   X=ScalarParameter() {defaultValue=-0.02, parameterName=Pan_U, sliderMax=1}
   Y=ScalarParameter() {defaultValue=1, parameterName=Pan_V, sliderMax=1}
  ) {materialFunction=&MakeFloat2}
  n18 = TextureCoordinate() {uTiling=0.4}
  n19 = Time()
  n20 = MaterialFunctionCall(
   X=ScalarParameter() {defaultValue=-0.02, parameterName=Pan_U, sliderMax=1}
   Y=ScalarParameter() {defaultValue=1, parameterName=Pan_V, sliderMax=1}
  ) {materialFunction=&MakeFloat2}
  n21 = Multiply(
   A=TextureSample(UVs=Panner(Coordinate=@n15, Time=@n16, Speed=@n17.Result) {speedX=1, speedY=1}) {texture=/Game/Maps/Textures/RadialGradient2}.RGB
   B=TextureSample(UVs=Panner(Coordinate=@n18, Time=@n19, Speed=@n20.Result) {speedX=1, speedY=1}) {texture=&RadialGradient2}.G
  )
  n22 = Panner(
   Coordinate=TextureCoordinate() {uTiling=0.6}
   Time=Time()
   Speed=MaterialFunctionCall(
    X=ScalarParameter() {defaultValue=-0.02, parameterName=Pan_U, sliderMax=1}
    Y=ScalarParameter() {defaultValue=1, parameterName=Pan_V, sliderMax=1}
   ) {materialFunction=&MakeFloat2}.Result
  ) {speedX=1, speedY=1}
  n23 = Multiply(A=Multiply(A=Multiply(A=@n21) {constB=2}, B=TextureSample(UVs=@n22) {texture=&RadialGradient2}.G)) {constB=3}
  n24 = Power(Base=MaterialFunctionCall() {materialFunction=&LinearGradient}.VGradient) {constExponent=3}
  n25 = MaterialFunctionCall("FadeSensitivity (optional)"=Constant() {r=2}) {materialFunction=&FadeBasedOnViewAngle}
  n26 = Sine(None=Multiply(A=Time()) {constB=0.3})
  n27 = LinearInterpolate(
   A=@n24
   B=Power(Base=MaterialFunctionCall() {materialFunction=&LinearGradient}.VGradient) {constExponent=6}
   Alpha=Multiply(A=Add(A=@n26)) {constB=0.5}
  )
  Multiply(A=ParticleColor().A, B=Multiply(A=Add(A=@n23, B=@n24), B=Multiply(A=@n25.Result, B=@n27)))
  NamedRerouteDeclaration(None=@n14) {name=Emissive, nodeColor=<rgba 1 0.1406 0 1>}
  n28 = Time()
  n29 = ScalarParameter() {defaultValue=0.1, parameterName=PanSpeed}
  NamedRerouteDeclaration(None=Add(A=DynamicParameter() {defaultValue=<rgba 0 1 1 0>, paramNames [Random, WPO Strength, NA, NA]}.Random, B=Multiply(A=@n28, B=@n29))) {name=Random UV Offset, nodeColor=<rgba 0.3359 0 1 1>}
  n30 = TextureCoordinate()
  n31 = TextureCoordinate()
  n32 = VectorParameter() {parameterName=ColorRibsUV TIle, defaultValue=<rgba 0.5 0.5 0 0>}
  NamedRerouteDeclaration(None=Multiply(A=StaticSwitchParameter(True=@n30, False=@n31) {defaultValue=true, parameterName=UseCoord2ForTiling?}, B=AppendVector(A=@n32.R, B=@n32.R))) {name=ColorRibs_UV, nodeColor=<rgba 0.6641 1 0 1>}
  n33 = VectorParameter() {parameterName=ColorDistortion_Tile, defaultValue=<rgba 2 0.12 0 0>}
  n34 = Multiply(A=NamedRerouteUsage(), B=AppendVector(A=@n33.R, B=@n33.R))
  n35 = NamedRerouteUsage()
  n36 = VectorParameter() {parameterName=ColorDistortion_Pan, defaultValue=<rgba 0.1 0.25 0 0>}
  n37 = AppendVector(A=@n36.R, B=@n36.R)
  n38 = Constant() {r=0.25}
  n39 = ShadingPathSwitch(
   Default=TextureSample(UVs=Panner(Coordinate=@n34, Time=@n35, Speed=@n37) {speedX=1, speedY=1}) {texture=/Game/Effects/Fort_Effects/Textures/Tile/T_Caustics_Chromatic}.RGB
   Forward=@n38
   Mobile=@n38
  )
  n40 = Constant() {r=0.25}
  NamedRerouteDeclaration(None=QualitySwitch(Default=ShadingPathSwitch(Default=@n39, Forward=@n40, Mobile=@n40), Low=@n40)) {name=Distortion, nodeColor=<rgba 0 1 0.7188 1>}
  NamedRerouteUsage()
 ]
}
