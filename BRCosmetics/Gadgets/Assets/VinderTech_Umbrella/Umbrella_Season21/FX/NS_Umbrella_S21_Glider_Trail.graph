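# Niagara trail effect for the Season 21 VinderTech umbrella glider, with cyan-blue glows, embers and drifting petals.
# A cone mesh, light and glow sprites follow the glider's FX_Cone socket, while a one-shot petal burst accompanies opening.
# GPU "WorldSpace" ember and petal emitters supplement their counterparts, inheriting source movement with distance falloff.
# User.bIsFrontEnd switches spawn offsets and drift velocities, and User.ForwardDot scales acceleration on the looping petals.
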NiagaraSystem /BRCosmetics/Gadgets/Assets/VinderTech_Umbrella/Umbrella_Season21/FX/NS_Umbrella_S21_Glider_Trail
dependencies {
 usedDataInterfaces [
  &NiagaraDataInterfaceCurve
  &NiagaraDataInterfaceColorCurve
  &NiagaraDataInterfaceVectorCurve
  &NiagaraDataInterfaceVector2DCurve
 ]
 usedDynamicInputs []
 usedModules [&SystemState, &FN_Integrator]
}
emitters {
 Cone_ScreenGlow001 {
  emitterUpdateScript [
   &EmitterState(
    Life Cycle Mode=::NewEnumerator1
    Loop Behavior=::NewEnumerator1
    Loop Duration Mode=::NewEnumerator1
    Scalability Mode=::NewEnumerator1
    UseLoopDelay=false
   )
   &SpawnBurst_Instantaneous(Spawn Count=1, Use Spawn Probability=false) {module=SpawnBurst_Instantaneous001}
   &SpawnBurst_Instantaneous(Spawn Count=1, Use Spawn Probability=false) {enabled=false}
   &SpawnRate(Spawn Probability=0.4, SpawnRate=0.35, Use Spawn Probability=false) {enabled=false}
  ]
  particleSpawnScript [
   &InitializeParticle(
    Color=<rgba 0.1082 0.663 1 1>
    Lifetime=4
    Mesh Scale=<1.5 1.5 3>
    Mesh Scale Mode=::NewEnumerator4
    Position Offset=<0 0 130>
    UsePositionOffset=true
    Write Material Random=true
   ) {module=InitializeParticle001}
   &SystemLocation(MaskSpawnGroup=false, Offset=<0 0 110>) {enabled=false}
   &DynamicMaterialParameters(
    Index 0 Param 1=0.5
    Module.Param0WriteEnabled=true
    Module.Param1WriteEnabled=true
    Module.Param2WriteEnabled=true
    Module.Param3WriteEnabled=true
   )
   &InitialMeshOrientation(
    Rotation=&RandomRangeVector(Minimum=<0 0 0>, Maximum=<0 0 0>)
    Rotation Coordinate Space=::NewEnumerator0
    Use Orientation Vector=true
    Use Rotation Vector=true
   )
  ]
  particleUpdateScript [
   &ParticleState(Kill Particles When Lifetime Has Elapsed=false, Let Infinitely Lived Particles Die When Emitter Deactivates=true)
   &SolveForcesAndVelocity(Clamp Velocity=false, Limit Acceleration=false) {enabled=false}
   &ScaleColor(
    Scale Alpha=&FloatFromCurve(FloatCurve=DataInterface() {propertyValues {Curve {keys [<at 0 0 RCIM_Cubic>,<at 0.5 3 RCIM_Cubic>,<at 1 0 RCIM_Cubic>]}}})
    Scale RGB=&VectorFromFloat(Value=&FloatFromCurve(FloatCurve=DataInterface() {propertyValues {Curve {keys [<at 0.5 0.1 RCIM_Cubic>]}}}, Scale Curve=4))
    ScaleA=true
    ScaleRGB=true
   ) {enabled=false}
   &SkeletalMeshLocation(
    MaskSpawnGroup=false
    Mesh Sampling Type=::NewEnumerator5
    Skeletal Mesh=DataInterface() {propertyValues {
      DefaultMesh=null
      ExcludeBoneName=root
      OptionalComponentClass=null
      PreviewMesh=/BRCosmetics/Gadgets/Assets/VinderTech_Umbrella/Umbrella_Season17/Meshes/Season_17_Glider
      SkinningMode=SkinOnTheFly
      SoftSourceActor=null
      SourceMode=Default
      UvSetIndex=0
      WholeMeshLOD=-1
      bExcludeBone=false
      bReadDeformedGeometry=true
      bRequireCurrentFrameData=true
      ComponentTags []
      FilteredBones []
      FilteredSockets [FX_LightCone]
      MeshUserParameter {parameter {name=None, typeDefHandle {registeredTypeIndex=1}, varData []}}
      SamplingRegions []
     }}
    Socket Sampling Mode=::NewEnumerator5
   ) {enabled=false}
   &SkeletalMeshLocation(
    MaskSpawnGroup=false
    Mesh Sampling Type=::NewEnumerator5
    Skeletal Mesh=DataInterface() {propertyValues {
      DefaultMesh=null
      ExcludeBoneName=root
      OptionalComponentClass=null
      PreviewMesh=/BRCosmetics/Gadgets/Assets/VinderTech_Umbrella/Umbrella_Season21/Meshes/Season_21_Glider
      SkinningMode=SkinOnTheFly
      SoftSourceActor=null
      SourceMode=Default
      UvSetIndex=0
      WholeMeshLOD=-1
      bExcludeBone=false
      bReadDeformedGeometry=true
      bRequireCurrentFrameData=true
      ComponentTags []
      FilteredBones []
      FilteredSockets [FX_Cone]
      MeshUserParameter {parameter {name=None, typeDefHandle {registeredTypeIndex=1}, varData []}}
      SamplingRegions []
     }}
    Socket Sampling Mode=::NewEnumerator5
   ) {module=SkeletalMeshLocation001}
  ]
  rendererClasses [&NiagaraMeshRendererProperties]
  renderers [{rendererClass=&NiagaraMeshRendererProperties, rendererIndex=0}]
 }
 Embers001 {
  emitterUpdateScript [
   &EmitterState(Life Cycle Mode=::NewEnumerator1, Loop Delay=0.8, Loop Duration=1, UseLoopDelay=false)
   &SpawnRate(Spawn Probability=0.4, SpawnRate=10, Use Spawn Probability=false)
  ]
  particleSpawnScript [
   &InitializeParticle(
    Lifetime Max=1
    Lifetime Min=1.5
    Lifetime Mode=::NewEnumerator1
    Sprite Size Mode=::NewEnumerator1
    Uniform Sprite Size Max=2.5
    Uniform Sprite Size Min=0.5
    UsePositionOffset=false
    Write Material Random=true
   ) {module=InitializeParticle001}
   /SetVariables(Particles.ScaleU:NiagaraFloat=&FloatFromCurve(
     FloatCurve=DataInterface() {propertyValues {Curve {keys [
         {arriveTangent=-1.6504, interpMode=RCIM_Cubic, leaveTangent=-1.6504, tangentMode=RCTM_User, time=-0.02, value=1.0266}
         <at 0.5 0.5>
         {arriveTangent=-2.0092, interpMode=RCIM_Cubic, leaveTangent=-2.0092, tangentMode=RCTM_User, time=2, value=0.35}
        ]}}}
     CurveIndex=$Emitter.Age
    )) {isSetParameters=true, module=SetVariables_#1}
   &Color(Color=<rgba 0.3713 2.2112 3 1>)
   &AddVelocity(Velocity=&RandomRangeVector(Minimum=<-5 -5 20>, Maximum=<5 5 10>, Evaluation Type=::NewEnumerator1))
   &SphereLocation(
    MaskSpawnGroup=false
    Override Seed=false
    Sphere Radius=2
    Surface Only=false
    UseNonUniformScale=false
    UseOffset=false
    UseSphereOrientation=false
   ) {enabled=false, module=SphereLocation001}
   &ShapeLocation(
    Cylinder Height=45
    Cylinder Radius=60
    MaskSpawnGroup=false
    Offset=&MakeCustomVectorFromBool(Bool=$User.bIsFrontEnd, True Vector=<0 0 35>, False Vector=<0 0 100>)
    Offset Mode=::NewEnumerator0
    Override Seed=false
    Shape Primitive=::NewEnumerator1
    Surface Only=false
    UseOwnerScale=true
   )
   &SkeletalMeshLocation(
    MaskSpawnGroup=false
    Mesh Sampling Type=::NewEnumerator5
    OverrideRandomSeed=false
    Skeletal Mesh=DataInterface() {propertyValues {
      DefaultMesh=null
      ExcludeBoneName=root
      OptionalComponentClass=null
      PreviewMesh=&Season_21_Glider
      SkinningMode=SkinOnTheFly
      SoftSourceActor=null
      SourceMode=Default
      UvSetIndex=0
      WholeMeshLOD=-1
      bExcludeBone=false
      bReadDeformedGeometry=true
      bRequireCurrentFrameData=true
      ComponentTags []
      FilteredBones []
      FilteredSockets [FX_Cone]
      MeshUserParameter {parameter {name=None, typeDefHandle {registeredTypeIndex=1}, varData []}}
      SamplingRegions []
     }}
   ) {enabled=false, module=SkeletalMeshLocation001}
  ]
  particleUpdateScript [
   &ParticleState()
   &CurlNoiseForce(Noise Strength=150, Pan Noise=false) {enabled=false}
   &TorusLocation(
    Align Torus With Axis=true
    Handle Radius=0
    Large Radius:NiagaraFloat=&FloatFromCurve(
     FloatCurve=DataInterface() {propertyValues {Curve {keys [{arriveTangent=-2.8702, interpMode=RCIM_Cubic, leaveTangent=-2.8702, tangentMode=RCTM_User, time=0, value=1}, <at 1 0>]}}}
     Scale Curve=&FloatFromCurve(
      FloatCurve=DataInterface() {propertyValues {Curve {keys [<at -0.0318 1 RCIM_Cubic>, {arriveTangent=-0.688, interpMode=RCIM_Cubic, leaveTangent=-0.688, tangentMode=RCTM_User, time=2.0112, value=0.3}]}}}
      CurveIndex=$Emitter.Age
      Scale Curve=&RandomRangeFloat(Minimum=100, Maximum=300)
     )
    )
    MaskSpawnGroup=false
    Radius Position=0
    Torus Axis=&RandomVector(Vector Scale=1)
    Torus Distribution Mode:ENiagaraTorusDistributionMode=::NewEnumerator1
    Torus Origin:NiagaraPosition=$Engine.Owner.Position
    U Position:NiagaraFloat=&Add_Float(
     A=&RandomRangeFloat(Minimum=0, Maximum=1)
     B=&Multiply_Float(
      A=&FloatFromCurve(FloatCurve=DataInterface() {propertyValues {Curve {keys [<at 0 1>,<at 1 0>]}}}, Scale Curve=0.5)
      B=$Particles.Initial.ScaleU
     )
    )
    V Position=1.3106
   ) {enabled=false}
   &Drag(Drag=&RandomRangeFloat(Minimum=1, Maximum=2), Use Linear Drag=true, Use Rotational Drag=false)
   &SolveForcesAndVelocity(Clamp Velocity=false, Limit Acceleration=false)
   &ScaleColor(
    Scale Alpha=&FloatFromCurve(FloatCurve=DataInterface() {propertyValues {Curve {keys [
         <at 0 0 RCIM_Cubic>
         {arriveTangent=-1.007, interpMode=RCIM_Cubic, leaveTangent=-1.007, tangentMode=RCTM_User, time=0.1, value=0.95}
         <at 0.2169 0.0158>
        ]}}})
    Scale RGB=&VectorFromCurve(VectorCurve=DataInterface() {propertyValues {
       XCurve {keys [
         {interpMode=RCIM_Cubic, tangentMode=RCTM_User, time=0, value=0}
         {arriveTangent=1080.0369, interpMode=RCIM_Cubic, leaveTangent=1080.0369, tangentMode=RCTM_User, time=0.1, value=50}
         {arriveTangent=-71.4286, interpMode=RCIM_Cubic, leaveTangent=-71.4286, time=0.3, value=899.9999}
         <at 0.8 0>
        ]}
       YCurve {keys [
         {interpMode=RCIM_Cubic, tangentMode=RCTM_User, time=0, value=0}
         {arriveTangent=1080.0369, interpMode=RCIM_Cubic, leaveTangent=1080.0369, tangentMode=RCTM_User, time=0.1, value=50}
         {arriveTangent=-71.4286, interpMode=RCIM_Cubic, leaveTangent=-71.4286, time=0.3, value=899.9999}
         <at 0.8 0>
        ]}
       ZCurve {keys [
         {interpMode=RCIM_Cubic, tangentMode=RCTM_User, time=0, value=0}
         {arriveTangent=1080.0369, interpMode=RCIM_Cubic, leaveTangent=1080.0369, tangentMode=RCTM_User, time=0.1, value=50}
         {arriveTangent=-71.4286, interpMode=RCIM_Cubic, leaveTangent=-71.4286, time=0.3, value=899.9999}
         <at 0.8 0>
        ]}
      }})
    ScaleA=true
    ScaleRGB=true
   ) {enabled=false}
   &ScaleSpriteSize(Scale Factor=&Multiply_Vector2(
     A=&Vector2DFromCurve(
      Vector2Curve=DataInterface() {propertyValues {
        XCurve {keys [<at 0 0 RCIM_Cubic>, {arriveTangent=-0.2212, interpMode=RCIM_Cubic, leaveTangent=-0.2212, time=0.2, value=2.375}, <at 0.8728 -0.1931>]}
        YCurve {keys [<at 0.0009 0 RCIM_Cubic>, {arriveTangent=-0.2215, interpMode=RCIM_Cubic, leaveTangent=-0.2215, time=0.2, value=2.375}, <at 0.8728 -0.1931>]}
       }}
      ScaleCurve=<1 1>
     )
     B=<6 6>
    ))
  ]
  rendererClasses [&NiagaraSpriteRendererProperties]
  renderers [{rendererClass=&NiagaraSpriteRendererProperties}]
 }
 Embers_WorldSpace {
  simTarget=GPUComputeSim
  emitterUpdateScript [
   &EmitterState(Life Cycle Mode=::NewEnumerator1, Loop Delay=0.8, Loop Duration=1, UseLoopDelay=false)
   &SpawnRate(Spawn Probability=0.4, SpawnRate=20, Use Spawn Probability=false)
  ]
  particleSpawnScript [
   &InitializeParticle(
    Lifetime Max=1
    Lifetime Min=0.5
    Lifetime Mode=::NewEnumerator1
    Sprite Size Mode=::NewEnumerator1
    Uniform Sprite Size Max=1
    Uniform Sprite Size Min=0.5
    UsePositionOffset=false
    Write Material Random=true
   ) {module=InitializeParticle001}
   /SetVariables(Particles.ScaleU:NiagaraFloat=&FloatFromCurve(
     FloatCurve=DataInterface() {propertyValues {Curve {keys [
         {arriveTangent=-1.6504, interpMode=RCIM_Cubic, leaveTangent=-1.6504, tangentMode=RCTM_User, time=-0.02, value=1.0266}
         <at 0.5 0.5>
         {arriveTangent=-2.0092, interpMode=RCIM_Cubic, leaveTangent=-2.0092, tangentMode=RCTM_User, time=2, value=0.35}
        ]}}}
     CurveIndex=$Emitter.Age
    )) {isSetParameters=true, module=SetVariables_#1}
   &Color(Color=<rgba 0.2475 1.4741 2 1>) {module=Color001}
   &AddVelocity(Velocity=&MakeCustomVectorFromBool(
     Bool=$User.bIsFrontEnd
     True Vector=<-100 -100 30>
     False Vector=&RandomRangeVector(Minimum=<-5 -5 -5>, Maximum=<5 5 10>)
    ))
   &SphereLocation(
    MaskSpawnGroup=false
    Override Seed=false
    Sphere Radius=2
    Surface Only=false
    UseNonUniformScale=false
    UseOffset=false
    UseSphereOrientation=false
   ) {enabled=false, module=SphereLocation001}
   &ShapeLocation(
    Cylinder Height=45
    Cylinder Radius=60
    MaskSpawnGroup=false
    Offset=&MakeCustomVectorFromBool(Bool=$User.bIsFrontEnd, True Vector=<0 0 35>, False Vector=<0 0 100>)
    Offset Mode=::NewEnumerator0
    Override Seed=false
    Shape Primitive=::NewEnumerator1
    Surface Only=false
    UseOwnerScale=true
   )
  ]
  particleUpdateScript [
   &ParticleState()
   &CurlNoiseForce(Noise Strength=150, Pan Noise=false) {enabled=false}
   &TorusLocation(
    Align Torus With Axis=true
    Handle Radius=0
    Large Radius:NiagaraFloat=&FloatFromCurve(
     FloatCurve=DataInterface() {propertyValues {Curve {keys [{arriveTangent=-2.8702, interpMode=RCIM_Cubic, leaveTangent=-2.8702, tangentMode=RCTM_User, time=0, value=1}, <at 1 0>]}}}
     Scale Curve=&FloatFromCurve(
      FloatCurve=DataInterface() {propertyValues {Curve {keys [<at -0.0318 1 RCIM_Cubic>, {arriveTangent=-0.688, interpMode=RCIM_Cubic, leaveTangent=-0.688, tangentMode=RCTM_User, time=2.0112, value=0.3}]}}}
      CurveIndex=$Emitter.Age
      Scale Curve=&RandomRangeFloat(Minimum=100, Maximum=300)
     )
    )
    MaskSpawnGroup=false
    Radius Position=0
    Torus Axis=&RandomVector(Vector Scale=1)
    Torus Distribution Mode:ENiagaraTorusDistributionMode=::NewEnumerator1
    Torus Origin:NiagaraPosition=$Engine.Owner.Position
    U Position:NiagaraFloat=&Add_Float(
     A=&RandomRangeFloat(Minimum=0, Maximum=1)
     B=&Multiply_Float(
      A=&FloatFromCurve(FloatCurve=DataInterface() {propertyValues {Curve {keys [<at 0 1>,<at 1 0>]}}}, Scale Curve=0.5)
      B=$Particles.Initial.ScaleU
     )
    )
    V Position=1.3106
   ) {enabled=false}
   &InheritSourceMovement(
    Applied Position Falloff Distance=2000
    Applied Position Scale=<0.99 0.99 1>
    ApplyForce=false
    ApplyPosition=true
    ApplyPositionFalloff=true
    ApplyPositionFalloffCurve=false
   )
   &Drag(Drag=&RandomRangeFloat(Minimum=1.5, Maximum=2.2), Use Linear Drag=true, Use Rotational Drag=false)
   &SolveForcesAndVelocity(Clamp Velocity=false, Limit Acceleration=false)
   &ScaleColor(
    Scale Alpha=&FloatFromCurve(FloatCurve=DataInterface() {propertyValues {Curve {keys [
         <at 0 0 RCIM_Cubic>
         {arriveTangent=-1.007, interpMode=RCIM_Cubic, leaveTangent=-1.007, tangentMode=RCTM_User, time=0.1, value=0.95}
         <at 0.2169 0.0158>
        ]}}})
    Scale RGB=&VectorFromCurve(VectorCurve=DataInterface() {propertyValues {
       XCurve {keys [
         {interpMode=RCIM_Cubic, tangentMode=RCTM_User, time=0, value=0}
         {arriveTangent=1080.0369, interpMode=RCIM_Cubic, leaveTangent=1080.0369, tangentMode=RCTM_User, time=0.1, value=50}
         {arriveTangent=-71.4286, interpMode=RCIM_Cubic, leaveTangent=-71.4286, time=0.3, value=899.9999}
         <at 0.8 0>
        ]}
       YCurve {keys [
         {interpMode=RCIM_Cubic, tangentMode=RCTM_User, time=0, value=0}
         {arriveTangent=1080.0369, interpMode=RCIM_Cubic, leaveTangent=1080.0369, tangentMode=RCTM_User, time=0.1, value=50}
         {arriveTangent=-71.4286, interpMode=RCIM_Cubic, leaveTangent=-71.4286, time=0.3, value=899.9999}
         <at 0.8 0>
        ]}
       ZCurve {keys [
         {interpMode=RCIM_Cubic, tangentMode=RCTM_User, time=0, value=0}
         {arriveTangent=1080.0369, interpMode=RCIM_Cubic, leaveTangent=1080.0369, tangentMode=RCTM_User, time=0.1, value=50}
         {arriveTangent=-71.4286, interpMode=RCIM_Cubic, leaveTangent=-71.4286, time=0.3, value=899.9999}
         <at 0.8 0>
        ]}
      }})
    ScaleA=true
    ScaleRGB=true
   ) {enabled=false}
   &ScaleSpriteSize(Scale Factor=&Multiply_Vector2(
     A=&Vector2DFromCurve(
      Vector2Curve=DataInterface() {propertyValues {
        XCurve {keys [<at 0 0 RCIM_Cubic>, {arriveTangent=-0.2212, interpMode=RCIM_Cubic, leaveTangent=-0.2212, time=0.2, value=2.375}, <at 0.8728 -0.1931>]}
        YCurve {keys [<at 0.0009 0 RCIM_Cubic>, {arriveTangent=-0.2215, interpMode=RCIM_Cubic, leaveTangent=-0.2215, time=0.2, value=2.375}, <at 0.8728 -0.1931>]}
       }}
      ScaleCurve=<1 1>
     )
     B=<6 6>
    ))
  ]
  rendererClasses [&NiagaraSpriteRendererProperties]
  renderers [{rendererClass=&NiagaraSpriteRendererProperties}]
 }
 Glow_and_Light001 {
  emitterUpdateScript [
   &EmitterState(
    Life Cycle Mode=::NewEnumerator1
    Loop Behavior=::NewEnumerator1
    Loop Duration Mode=::NewEnumerator1
    Scalability Mode=::NewEnumerator1
    UseLoopDelay=false
   )
   &SpawnBurst_Instantaneous(Spawn Count=1, Use Spawn Probability=false) {module=SpawnBurst_Instantaneous001}
   &SpawnBurst_Instantaneous(Spawn Count=1, Use Spawn Probability=false) {enabled=false}
   &SpawnRate(SpawnRate=0.35, Use Spawn Probability=false) {enabled=false}
  ]
  particleSpawnScript [
   /Niagara/Modules/Spawn/Initialization/InitializeParticle(
    Color=<rgba 1 0.1812 0.9082 1>
    Lifetime=4
    Module.Write Color=true
    Module.Write Lifetime=true
    Module.Write Mass=false
    Module.Write Material Random=true
    Module.Write Position=false
    Module.Write Scale=false
    Module.Write SpriteRotation=true
    Module.Write SpriteSize=true
    Sprite Rotation=&RandomRangeFloat(Minimum=0, Maximum=360)
   )
   &SystemLocation(MaskSpawnGroup=false, Offset=<0 0 140>)
   &DynamicMaterialParameters(
    Index 0 Param 1=0.5
    Module.Param0WriteEnabled=true
    Module.Param1WriteEnabled=true
    Module.Param2WriteEnabled=true
    Module.Param3WriteEnabled=true
   )
  ]
  particleUpdateScript [
   &ParticleState(Kill Particles When Lifetime Has Elapsed=false, Let Infinitely Lived Particles Die When Emitter Deactivates=true)
   &SolveForcesAndVelocity(Clamp Velocity=false, Limit Acceleration=false) {enabled=false}
   &ScaleColor(
    Scale RGB=&VectorFromFloat(Value=&FloatFromCurve(
      FloatCurve=DataInterface() {propertyValues {Curve {keys [{interpMode=RCIM_Cubic, tangentMode=RCTM_User, time=0, value=0}, <at 0.5 1 RCIM_Cubic>, <at 1 0>]}}}
      Scale Curve=1.5
     ))
    ScaleA=true
    ScaleRGB=true
   ) {enabled=false}
   /SetVariables(Particles.LightColor=&ScaleLinearColorByCurve(
     InitialValue=<rgba 0.1028 0.7684 1 15>
     ScaleCurve=DataInterface() {propertyValues {AlphaCurve {keys [<at 0.4948 1>]}, BlueCurve {keys [<at 1 1>]}, GreenCurve {keys [<at 1 1>]}, RedCurve {keys [<at 1 1>]}}}
     CurveIndex=$Particles.NormalizedAge
    )) {isSetParameters=true}
   &ScaleSpriteSize(Scale Factor=&Vector2DFromFloat(Value=&FloatFromCurve(FloatCurve=DataInterface() {propertyValues {Curve {keys [
          {interpMode=RCIM_Cubic, tangentMode=RCTM_User, time=0, value=0}
          {interpMode=RCIM_Cubic, tangentMode=RCTM_User, time=0.31, value=1}
          {interpMode=RCIM_Cubic, tangentMode=RCTM_User, time=0.7, value=1}
          {interpMode=RCIM_Cubic, tangentMode=RCTM_User, time=1, value=0}
         ]}}}))) {enabled=false}
   &CameraOffset(Camera Offset Amount=-30)
   &SkeletalMeshLocation(
    MaskSpawnGroup=false
    Mesh Sampling Type=::NewEnumerator5
    Skeletal Mesh=DataInterface() {propertyValues {
      DefaultMesh=null
      ExcludeBoneName=root
      OptionalComponentClass=null
      PreviewMesh=&Season_17_Glider
      SkinningMode=SkinOnTheFly
      SoftSourceActor=null
      SourceMode=Default
      UvSetIndex=0
      WholeMeshLOD=-1
      bExcludeBone=false
      bReadDeformedGeometry=true
      bRequireCurrentFrameData=true
      ComponentTags []
      FilteredBones []
      FilteredSockets [FX_LightCone]
      MeshUserParameter {parameter {name=None, typeDefHandle {registeredTypeIndex=1}, varData []}}
      SamplingRegions []
     }}
    Socket Sampling Mode=::NewEnumerator5
   ) {enabled=false}
   &SkeletalMeshLocation(
    MaskSpawnGroup=false
    Mesh Sampling Type=::NewEnumerator5
    OverrideRandomSeed=false
    Sampled Position Offset=<0 0 10>
    Skeletal Mesh=DataInterface() {propertyValues {
      DefaultMesh=null
      ExcludeBoneName=root
      OptionalComponentClass=null
      PreviewMesh=&Season_21_Glider
      SkinningMode=SkinOnTheFly
      SoftSourceActor=null
      SourceMode=Default
      UvSetIndex=0
      WholeMeshLOD=-1
      bExcludeBone=false
      bReadDeformedGeometry=true
      bRequireCurrentFrameData=true
      ComponentTags []
      FilteredBones []
      FilteredSockets [FX_Cone]
      MeshUserParameter {parameter {name=None, typeDefHandle {registeredTypeIndex=1}, varData []}}
      SamplingRegions []
     }}
   ) {module=SkeletalMeshLocation001}
  ]
  rendererClasses [&NiagaraLightRendererProperties, &NiagaraSpriteRendererProperties]
  renderers [
   {rendererClass=&NiagaraLightRendererProperties, rendererIndex=0}
   {rendererClass=&NiagaraSpriteRendererProperties, rendererIndex=1}
  ]
 }
 Main_Glow {
  emitterUpdateScript [
   &EmitterState()
   &SpawnRate(SpawnRate=0.4, Use Spawn Probability=false)
  ]
  particleSpawnScript [
   /Niagara/Modules/Spawn/Initialization/V2/InitializeParticle(
    Lifetime=6
    Position Offset=<0 0 100>
    Sprite Size Mode=::NewEnumerator3
    Uniform Sprite Size=350
    UsePositionOffset=true
    Write Material Random=true
   ) {module=InitializeParticle001}
   &Color(Color=<rgba 0.0631 0.3466 1 0.8>)
   &DynamicMaterialParameters(
    Index 0 Param 1=10
    Module.Param0WriteEnabled=true
    Module.Param1WriteEnabled=false
    Module.Param2WriteEnabled=false
    Module.Param3WriteEnabled=false
   ) {enabled=false}
   &BoxLocation(Box Offset=<0 -20 0>, Box Size=<0 0 0>, MaskSpawnGroup=false, Module.Override Seed=false, Rotate Box=false, Surface Only=false) {enabled=false, module=BoxLocation001}
  ]
  particleUpdateScript [
   &ParticleState()
   &ScaleColor(
    Scale Alpha=&FloatFromCurve(FloatCurve=DataInterface() {propertyValues {Curve {keys [<at 0 0 RCIM_Cubic>,<at 0.2 1 RCIM_Cubic>,<at 1 0>]}}})
    Scale RGB=<1.5 1.5 1.2>
    ScaleA=true
    ScaleRGB=true
   )
   &CameraOffset(Camera Offset Amount=5) {enabled=false}
   &ScaleSpriteSize(Uniform Curve Sprite Scale=DataInterface() {propertyValues {Curve {keys [{tangentMode=RCTM_User, time=0, value=0.8}, {tangentMode=RCTM_User, time=1, value=1}]}}})
   &SkeletalMeshLocation(
    MaskSpawnGroup=false
    Mesh Sampling Type=::NewEnumerator5
    OverrideRandomSeed=false
    Sampled Position Offset=<0 0 -30>
    Skeletal Mesh=DataInterface() {propertyValues {
      DefaultMesh=null
      ExcludeBoneName=root
      OptionalComponentClass=null
      PreviewMesh=&Season_21_Glider
      SkinningMode=SkinOnTheFly
      SoftSourceActor=null
      SourceMode=Default
      UvSetIndex=0
      WholeMeshLOD=-1
      bExcludeBone=false
      bReadDeformedGeometry=true
      bRequireCurrentFrameData=true
      ComponentTags []
      FilteredBones []
      FilteredSockets [FX_Cone]
      MeshUserParameter {parameter {name=None, typeDefHandle {registeredTypeIndex=1}, varData []}}
      SamplingRegions []
     }}
   ) {module=SkeletalMeshLocation001}
  ]
  rendererClasses [&NiagaraSpriteRendererProperties]
  renderers [
   {rendererClass=&NiagaraSpriteRendererProperties}
   {rendererClass=&NiagaraSpriteRendererProperties, rendererIndex=1}
  ]
 }
 Petals {
  emitterUpdateScript [
   &EmitterState(Life Cycle Mode=::NewEnumerator1, Loop Delay=0.6, Loop Duration=1, UseLoopDelay=false)
   &SpawnRate(Spawn Probability=0.3, SpawnRate=40, Use Spawn Probability=true)
  ]
  particleSpawnScript [
   /Niagara/Modules/Spawn/Initialization/InitializeParticle(
    Color=<rgba 0 0 0 1>
    Lifetime=&RandomRangeFloat(Minimum=3, Maximum=4)
    Module.Write Color=true
    Module.Write Lifetime=true
    Module.Write Mass=true
    Module.Write Material Random=true
    Module.Write Position=false
    Module.Write Scale=false
    Module.Write SpriteRotation=true
    Module.Write SpriteSize=true
    Position=&SimulationPosition()
    Sprite Rotation=&UniformRangedFloat(Minimum=0, Maximum=360)
    Sprite Size=&Multiply_Vector2(A=<2.5 2.5>, B=&UniformRangedVector2D(Minimum=<1 0.5>, Maximum=<2.5 1.6>))
   )
   &SphereLocation(
    MaskSpawnGroup=false
    Non Uniform Scale=<1 1 0.4>
    Offset=&MakeCustomVectorFromBool(Bool=$User.bIsFrontEnd, True Vector=<0 0 35>, False Vector=<0 0 100>)
    Override Seed=false
    Sphere Radius=60
    Surface Only=false
    UseNonUniformScale=true
    UseOffset=true
    UseSphereOrientation=false
   )
   &AddVelocity(Velocity=&RandomRangeVector(Minimum=<-10 -10 -10>, Maximum=<10 10 20>))
   &InheritVelocity(Module.Limit Speed=true, Velocity Limit=75, Velocity Scale=<-1 -1 0>) {enabled=false}
   /SetVariables(Particles.stableGravity:NiagaraFloat=&UniformRangedFloat(Minimum=5, Maximum=10)) {isSetParameters=true, module=SetVariables_#2}
   /SetVariables(Particles.stableRotationRate:NiagaraFloat=&UniformRangedFloat(Minimum=50, Maximum=200)) {isSetParameters=true, module=SetVariables_#3}
   &AddVelocityFromPoint(Module.Use Falloff Distance=false, Origin Offset Coordinate Space=::Simulation, Velocity Strength=20) {enabled=false}
   &SkeletalMeshLocation(
    Bone Sampling Mode=::NewEnumerator0
    Inherit Velocity=false
    MaskSpawnGroup=false
    Module.Override Randomness=false
    Skeletal Mesh=$User.SkeletalMesh
   ) {enabled=false}
  ]
  particleUpdateScript [
   /SetVariables(Particles.DistanceThreshold:NiagaraFloat=&MakeCustomFloatFromBool(
     Bool=&SetBoolByFloatComparison(
      A=0.05
      B=&Power(
       a=&ClampFloat(Float=&Subtract_Float(
         A=&Multiply_Float(A=9.1, B=&OneMinusFloat(Float=$Emitter.Scalability.DistanceFraction))
         B=8
        ))
       b=4
      )
      Comparison Type=::NewEnumerator6
     )
     True Float=&Power(
      a=&ClampFloat(Float=&Subtract_Float(
        A=&Multiply_Float(A=9.1, B=&OneMinusFloat(Float=$Emitter.Scalability.DistanceFraction))
        B=8
       ))
      b=4
     )
     False Float=0
    )) {enabled=false, isSetParameters=true, module=SetVariables_#4}
   &ParticleState()
   &GravityForce(Gravity=&MakeVector(X=0, Y=0, Z=$Particles.Initial.stableGravity))
   &Drag(Drag=&UniformRangedFloat(Minimum=0.8, Maximum=1.25), Use Linear Drag=true, Use Rotational Drag=false)
   /Niagara/Modules/Update/Forces/CurlNoiseForce(Noise Frequency=0.1, Noise Strength=80, Pan Noise=true, Pan Noise Field=<0 0 -50>)
   &VortexForce(UseOriginOffset=false, Vortex Force Amount=50)
   &AccelerationForce(Acceleration=&Lerp_Vector(A=<0 0 0>, B=<-20 4 0>, Alpha=$User.ForwardDot), Coordinate Space=::Local)
   &SolveForcesAndVelocity(Clamp Velocity=false, Limit Acceleration=false)
   &SpriteRotationRate(Rotation Rate=$Particles.Initial.stableRotationRate)
   &ScaleSpriteSize(Scale Factor=&Multiply_Vector2(
     A=&Multiply_Vector2(
      A=&MakeVector2D(
       X=&Multiply_Float(
        A=1
        B=&Sine(
         Normalized Angle=$Emitter.Age
         Period=&Multiply_Float(A=3, B=&Add_Float(A=$Particles.Initial.MaterialRandom, B=0.6))
         Scale=1
         Bias=0
        )
       )
       Y=1
      )
      B=&Vector2DFromFloat(Value=&FloatFromCurve(FloatCurve=DataInterface() {propertyValues {Curve {keys [<at 0 1.025>,<at 0.9 1.025>,<at 1 0>]}}}))
     )
     B=&Vector2DFromFloat(Value=1)
    ))
  ]
  rendererClasses [&NiagaraSpriteRendererProperties]
  renderers [{rendererClass=&NiagaraSpriteRendererProperties}]
 }
 Petals_Opening {
  emitterUpdateScript [
   &EmitterState(Life Cycle Mode=::NewEnumerator1, Loop Behavior=::NewEnumerator1, Loop Duration=3, UseLoopDelay=false)
   &SpawnRate(Spawn Probability=0.3, SpawnRate=40, Use Spawn Probability=true) {enabled=false}
   &SpawnBurst_Instantaneous(Spawn Count=100, Use Spawn Probability=false, UseLoopCountLimit=false)
  ]
  particleSpawnScript [
   &InitializeParticle(
    Color=<rgba 0 0 0 1>
    Lifetime=&RandomRangeFloat(Minimum=0.8, Maximum=2)
    Module.Write Color=true
    Module.Write Lifetime=true
    Module.Write Mass=true
    Module.Write Material Random=true
    Module.Write Position=false
    Module.Write Scale=false
    Module.Write SpriteRotation=true
    Module.Write SpriteSize=true
    Position=&SimulationPosition()
    Sprite Rotation=&UniformRangedFloat(Minimum=0, Maximum=360)
    Sprite Size=&Multiply_Vector2(A=<2.5 2.5>, B=&UniformRangedVector2D(Minimum=<1 0.5>, Maximum=<2.5 1.6>))
   )
   &SphereLocation(
    MaskSpawnGroup=false
    Non Uniform Scale=<1 1 0.4>
    Offset=&MakeCustomVectorFromBool(Bool=$User.bIsFrontEnd, True Vector=<0 0 35>, False Vector=<0 -20 100>)
    Override Seed=false
    Sphere Radius=20
    Surface Only=false
    UseNonUniformScale=true
    UseOffset=true
    UseSphereOrientation=false
   )
   &AddVelocity(Velocity=&RandomRangeVector(Minimum=<-200 -200 -150>, Maximum=<200 200 250>))
   /SetVariables(Particles.stableGravity:NiagaraFloat=&UniformRangedFloat(Minimum=15, Maximum=10)) {isSetParameters=true, module=SetVariables_#2}
   /SetVariables(Particles.stableRotationRate:NiagaraFloat=&UniformRangedFloat(Minimum=50, Maximum=200)) {isSetParameters=true, module=SetVariables_#3}
   &AddVelocityFromPoint(Module.Use Falloff Distance=false, Origin Offset Coordinate Space=::Simulation, Velocity Strength=20) {enabled=false}
   &SkeletalMeshLocation(
    Bone Sampling Mode=::NewEnumerator0
    Inherit Velocity=false
    MaskSpawnGroup=false
    Module.Override Randomness=false
    Skeletal Mesh=$User.SkeletalMesh
   ) {enabled=false}
  ]
  particleUpdateScript [
   &InheritVelocity(Module.Limit Speed=false, Velocity Limit=75) {enabled=false}
   /SetVariables(Particles.DistanceThreshold:NiagaraFloat=&MakeCustomFloatFromBool(
     Bool=&SetBoolByFloatComparison(
      A=0.05
      B=&Power(
       a=&ClampFloat(Float=&Subtract_Float(
         A=&Multiply_Float(A=9.1, B=&OneMinusFloat(Float=$Emitter.Scalability.DistanceFraction))
         B=8
        ))
       b=4
      )
      Comparison Type=::NewEnumerator6
     )
     True Float=&Power(
      a=&ClampFloat(Float=&Subtract_Float(
        A=&Multiply_Float(A=9.1, B=&OneMinusFloat(Float=$Emitter.Scalability.DistanceFraction))
        B=8
       ))
      b=4
     )
     False Float=0
    )) {enabled=false, isSetParameters=true, module=SetVariables_#4}
   &ParticleState()
   &GravityForce(Gravity=&MakeVector(X=0, Y=0, Z=$Particles.Initial.stableGravity))
   &InheritSourceMovement(ApplyForce=false, ApplyPosition=true, ApplyPositionFalloff=false)
   &Drag(Drag=&UniformRangedFloat(Minimum=1, Maximum=2), Use Linear Drag=true, Use Rotational Drag=false)
   &CurlNoiseForce(Noise Frequency=0.1, Noise Strength=300, Pan Noise=true, Pan Noise Field=<0 0 -50>)
   &VortexForce(UseOriginOffset=false)
   &AccelerationForce(Acceleration=&Lerp_Vector(A=<0 0 0>, B=<-20 4 0>, Alpha=$User.ForwardDot), Coordinate Space=::Local) {enabled=false}
   &SolveForcesAndVelocity(Clamp Velocity=false, Limit Acceleration=false)
   &SpriteRotationRate(Rotation Rate=$Particles.Initial.stableRotationRate)
   &ScaleSpriteSize(Scale Factor=&Multiply_Vector2(
     A=&Multiply_Vector2(
      A=&MakeVector2D(
       X=&Multiply_Float(
        A=1
        B=&Sine(
         Normalized Angle=$Emitter.Age
         Period=&Multiply_Float(A=3, B=&Add_Float(A=$Particles.Initial.MaterialRandom, B=0.6))
         Scale=1
         Bias=0
        )
       )
       Y=1
      )
      B=&Vector2DFromFloat(Value=&FloatFromCurve(FloatCurve=DataInterface() {propertyValues {Curve {keys [<at 0 1.025>, {arriveTangent=-1.025, interpMode=RCIM_Cubic, leaveTangent=-1.025, time=0.6, value=1.05}, <at 1 0>]}}}))
     )
     B=&Vector2DFromFloat(Value=1)
    ))
  ]
  rendererClasses [&NiagaraSpriteRendererProperties]
  renderers [{rendererClass=&NiagaraSpriteRendererProperties}]
 }
 Petals_WorldSpace {
  simTarget=GPUComputeSim
  emitterUpdateScript [
   &EmitterState(Life Cycle Mode=::NewEnumerator1, Loop Delay=0.6, Loop Duration=1, UseLoopDelay=false)
   &SpawnRate(Spawn Probability=0.6, SpawnRate=20, Use Spawn Probability=true)
  ]
  particleSpawnScript [
   &InitializeParticle(
    Color=<rgba 0 0 0 1>
    Lifetime=&RandomRangeFloat(Minimum=3, Maximum=2)
    Module.Write Color=true
    Module.Write Lifetime=true
    Module.Write Mass=true
    Module.Write Material Random=true
    Module.Write Position=false
    Module.Write Scale=false
    Module.Write SpriteRotation=true
    Module.Write SpriteSize=true
    Position=&SimulationPosition()
    Sprite Rotation=&UniformRangedFloat(Minimum=0, Maximum=360)
    Sprite Size=&Multiply_Vector2(A=<2.5 2.5>, B=&UniformRangedVector2D(Minimum=<1 0.5>, Maximum=<2.5 1.6>))
   )
   &SphereLocation(
    MaskSpawnGroup=false
    Non Uniform Scale=<1 1 0.4>
    Offset=&MakeCustomVectorFromBool(Bool=$User.bIsFrontEnd, True Vector=<0 0 35>, False Vector=<0 0 100>)
    Override Seed=false
    Sphere Radius=60
    Surface Only=false
    UseNonUniformScale=true
    UseOffset=true
    UseSphereOrientation=false
   )
   &AddVelocity(Velocity=&MakeCustomVectorFromBool(
     Bool=$User.bIsFrontEnd
     True Vector=<-150 -150 30>
     False Vector=&RandomRangeVector(Minimum=<-10 -10 -10>, Maximum=<10 10 20>)
    ))
   &InheritVelocity(Module.Limit Speed=true, Velocity Limit=75, Velocity Scale=<-1 -1 0>) {enabled=false}
   /SetVariables(Particles.stableGravity:NiagaraFloat=&UniformRangedFloat(Minimum=5, Maximum=10)) {isSetParameters=true, module=SetVariables_#2}
   /SetVariables(Particles.stableRotationRate:NiagaraFloat=&UniformRangedFloat(Minimum=50, Maximum=200)) {isSetParameters=true, module=SetVariables_#3}
   &AddVelocityFromPoint(Module.Use Falloff Distance=false, Origin Offset Coordinate Space=::Simulation, Velocity Strength=20) {enabled=false}
   &SkeletalMeshLocation(
    Bone Sampling Mode=::NewEnumerator0
    Inherit Velocity=false
    MaskSpawnGroup=false
    Module.Override Randomness=false
    Skeletal Mesh=$User.SkeletalMesh
   ) {enabled=false}
  ]
  particleUpdateScript [
   /SetVariables(Particles.DistanceThreshold:NiagaraFloat=&MakeCustomFloatFromBool(
     Bool=&SetBoolByFloatComparison(
      A=0.05
      B=&Power(
       a=&ClampFloat(Float=&Subtract_Float(
         A=&Multiply_Float(A=9.1, B=&OneMinusFloat(Float=$Emitter.Scalability.DistanceFraction))
         B=8
        ))
       b=4
      )
      Comparison Type=::NewEnumerator6
     )
     True Float=&Power(
      a=&ClampFloat(Float=&Subtract_Float(
        A=&Multiply_Float(A=9.1, B=&OneMinusFloat(Float=$Emitter.Scalability.DistanceFraction))
        B=8
       ))
      b=4
     )
     False Float=0
    )) {enabled=false, isSetParameters=true, module=SetVariables_#4}
   &ParticleState()
   &GravityForce(Gravity=&MakeVector(X=0, Y=0, Z=$Particles.Initial.stableGravity))
   &Drag(Drag=&UniformRangedFloat(Minimum=0.8, Maximum=1.25), Use Linear Drag=true, Use Rotational Drag=false)
   &CurlNoiseForce(Noise Frequency=0.1, Noise Strength=80, Pan Noise=true, Pan Noise Field=<0 0 -50>)
   &InheritSourceMovement(
    Applied Position Falloff Distance=3000
    Applied Position Scale=<0.99 0.99 1>
    ApplyForce=false
    ApplyPosition=true
    ApplyPositionFalloff=true
    ApplyPositionFalloffCurve=false
   )
   &VortexForce(UseOriginOffset=false, Vortex Force Amount=50)
   &AccelerationForce(Acceleration=&Lerp_Vector(A=<0 0 0>, B=<-20 4 0>, Alpha=$User.ForwardDot), Coordinate Space=::Local)
   &SolveForcesAndVelocity(Clamp Velocity=false, Limit Acceleration=false)
   &SpriteRotationRate(Rotation Rate=$Particles.Initial.stableRotationRate)
   &ScaleSpriteSize(Scale Factor=&Multiply_Vector2(
     A=&Multiply_Vector2(
      A=&MakeVector2D(
       X=&Multiply_Float(
        A=1
        B=&Sine(
         Normalized Angle=$Emitter.Age
         Period=&Multiply_Float(A=3, B=&Add_Float(A=$Particles.Initial.MaterialRandom, B=0.6))
         Scale=1
         Bias=0
        )
       )
       Y=1
      )
      B=&Vector2DFromFloat(Value=&FloatFromCurve(FloatCurve=DataInterface() {propertyValues {Curve {keys [<at 0 1.025>,<at 0.9 1.025>,<at 1 0>]}}}))
     )
     B=&Vector2DFromFloat(Value=1)
    ))
  ]
  rendererClasses [&NiagaraSpriteRendererProperties]
  renderers [{rendererClass=&NiagaraSpriteRendererProperties}]
 }
}
summary {
 systemName=NS_Umbrella_S21_Glider_Trail
 emitters [
  {emitterName=Cone_ScreenGlow001, rendererClasses [&NiagaraMeshRendererProperties]}
  {
   emitterName=Glow_and_Light001
   rendererClasses [&NiagaraLightRendererProperties, &NiagaraSpriteRendererProperties]
  }
  {emitterName=Main_Glow, rendererClasses [&NiagaraSpriteRendererProperties]}
  {emitterName=Embers001, rendererClasses [&NiagaraSpriteRendererProperties]}
  {emitterName=Petals, rendererClasses [&NiagaraSpriteRendererProperties]}
  {emitterName=Petals_Opening, rendererClasses [&NiagaraSpriteRendererProperties]}
  {emitterName=Petals_WorldSpace, simTarget=GPUComputeSim, rendererClasses [&NiagaraSpriteRendererProperties]}
  {emitterName=Embers_WorldSpace, simTarget=GPUComputeSim, rendererClasses [&NiagaraSpriteRendererProperties]}
 ]
 userVariables [
  User.bIsFrontEnd {type=NiagaraBool}
  User.bIsFrontEndPreview {type=NiagaraBool}
  User.bIsFullyDeployed {type=NiagaraBool}
  User.ForwardDot {type=NiagaraFloat, defaultValue=0, description=""}
  User.RibbonIDOffset {type=NiagaraInt32}
  User.RightDot {type=NiagaraFloat, defaultValue=0, description=""}
  User.SkeletalMesh {
   type=NiagaraDataInterfaceSkeletalMesh
   defaultValue {
    struct=NiagaraExt_VariableValue_DataInterface
    value {dataInterface=&self:NiagaraDataInterfaceSkeletalMesh_0}
   }
  }
 ]
}
