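# Masked, two-sided subsurface material for the branches of the Season 21 VinderTech umbrella gadget, usable on skeletal meshes and mesh deformers.
# It is a base Material despite the MI_ prefix.
# Its nodes are samples of the Leaves diffuse, normal and packed mask textures, plus a purple vector parameter.
# The graph lists no connections, so the role of each node in the shading is not shown.
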Material /BRCosmetics/Gadgets/Assets/VinderTech_Umbrella/Umbrella_Season21/Materials/MI_Umbrella_Season21_Branches
graph {
 properties {
  bUsedWithMeshDeformer=true
  bUsedWithSkeletalMesh=true
  blendMode=BLEND_Masked
  pixelDepthOffsetMode=PDOM_Legacy
  shadingModel=MSM_Subsurface
  twoSided=true
  usedShadingModels=Subsurface
 }
 nodes [
  TextureSample() {texture=/BRCosmetics/Gadgets/Assets/VinderTech_Umbrella/Umbrella_Season21/Textures/T_Umbrella_Season_21_Leaves_D}
  TextureSample() {
   samplerType=SAMPLERTYPE_Normal
   texture=/BRCosmetics/Gadgets/Assets/VinderTech_Umbrella/Umbrella_Season21/Textures/T_Umbrella_Season_21_Leaves_N
  }
  TextureSample() {
   samplerType=SAMPLERTYPE_Masks
   texture=/BRCosmetics/Gadgets/Assets/VinderTech_Umbrella/Umbrella_Season21/Textures/T_Umbrella_Season_21_Leaves_S
  }
  VectorParameter() {defaultValue=<rgba 0.4153 0 0.4531 0>}
 ]
}
